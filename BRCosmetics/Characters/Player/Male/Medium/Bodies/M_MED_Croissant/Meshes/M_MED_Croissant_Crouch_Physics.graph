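# Physics asset for the loincloth of the male medium Croissant player body, apparently the crouch variant judging by its name.
# Left and right loincloths are each a five-segment capsule chain hanging from a small base capsule.
# Identical angular limits link each segment to the one above, with wide swing on one axis and tight twist.
# Unconstrained thigh capsules are the only other bodies, likely serving as leg colliders for the cloth.
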PhysicsAsset /BRCosmetics/Characters/Player/Male/Medium/Bodies/M_MED_Croissant/Meshes/M_MED_Croissant_Crouch_Physics
bodies {
 dyn_loincloth_01_l [{shapeType=Capsule, length=3, radius=4, shapeName=dyn_loincloth_01_l_capsule, center=<6 0 0>, rotation=<rot -90 0 0>}]
 dyn_loincloth_01_r [{shapeType=Capsule, length=3, radius=4, shapeName=dyn_loincloth_01_r_capsule, center=<6 0 0>, rotation=<rot -90 0 0>}]
 dyn_loincloth_02_l [{
   shapeType=Capsule
   length=6.155
   radius=3.845
   shapeName=dyn_loincloth_02_l_capsule
   center=<5.0411 -0.0002 0.0881>
   rotation=<rot 88.9985 179.9999 180>
  }]
 dyn_loincloth_02_r [{
   shapeType=Capsule
   length=6.451
   radius=3.9235
   shapeName=dyn_loincloth_02_l_capsule
   center=<5.0411 -0.0002 0.0881>
   rotation=<rot 88.9985 179.9999 180>
  }]
 dyn_loincloth_03_l [{
   shapeType=Capsule
   length=6.6928
   radius=3.3072
   shapeName=dyn_loincloth_03_l_capsule
   center=<5.0411 -0.0002 0.0877>
   rotation=<rot 89.0034 179.9999 180>
  }]
 dyn_loincloth_03_r [{
   shapeType=Capsule
   length=6.3766
   radius=3.6234
   shapeName=dyn_loincloth_03_r_capsule
   center=<5.0411 -0.0002 0.0877>
   rotation=<rot 89.0034 179.9999 180>
  }]
 dyn_loincloth_04_l [{
   shapeType=Capsule
   length=6.889
   radius=3.111
   shapeName=dyn_loincloth_04_l_capsule
   center=<4.7374 -0.0001 0.0824>
   rotation=<rot 89.0034 179.9999 -180>
  }]
 dyn_loincloth_04_r [{
   shapeType=Capsule
   length=6.7119
   radius=3.2881
   shapeName=dyn_loincloth_04_r_capsule
   center=<4.7374 -0.0001 0.0828>
   rotation=<rot 88.9985 179.9999 -180>
  }]
 dyn_loincloth_05_l [{
   shapeType=Capsule
   length=6.43
   radius=3.57
   shapeName=dyn_loincloth_05_l_capsule
   center=<4.7374 -0.0001 0.0824>
   rotation=<rot 89.0034 179.9999 -180>
  }]
 dyn_loincloth_05_r [{
   shapeType=Capsule
   length=7.1131
   radius=2.8869
   shapeName=dyn_loincloth_05_r_capsule
   center=<4.7374 -0.0001 0.0824>
   rotation=<rot 89.0034 179.9999 -180>
  }]
 dyn_loincloth_base_l [{shapeType=Capsule, length=0.505, radius=0.505, shapeName=dyn_loincloth_base_l_capsule}]
 dyn_loincloth_base_r [{shapeType=Capsule, length=0.505, radius=0.505, shapeName=dyn_loincloth_base_r_capsule}]
 thigh_l [{shapeType=Capsule, length=25.4328, radius=7.5874, center=<-15.5961 0 -1.318>, rotation=<rot 90 0 0>}]
 thigh_r [{shapeType=Capsule, length=25.948, radius=8.082, shapeName=thigh_twist_01_r_capsule, center=<15.0294 0 0.9639>, rotation=<rot 90 0 0>}]
}
constraints [
 {bone1Name=dyn_loincloth_01_r, bone2Name=dyn_loincloth_base_r, swing1LimitDegrees=45, swing2LimitDegrees=15, twistLimitDegrees=5}
 {bone1Name=dyn_loincloth_02_r, bone2Name=dyn_loincloth_01_r, swing1LimitDegrees=45, swing2LimitDegrees=15, twistLimitDegrees=5}
 {bone1Name=dyn_loincloth_03_r, bone2Name=dyn_loincloth_02_r, swing1LimitDegrees=45, swing2LimitDegrees=15, twistLimitDegrees=5}
 {bone1Name=dyn_loincloth_04_r, bone2Name=dyn_loincloth_03_r, swing1LimitDegrees=45, swing2LimitDegrees=15, twistLimitDegrees=5}
 {bone1Name=dyn_loincloth_05_r, bone2Name=dyn_loincloth_04_r, swing1LimitDegrees=45, swing2LimitDegrees=15, twistLimitDegrees=5}
 {bone1Name=dyn_loincloth_01_l, bone2Name=dyn_loincloth_base_l, swing1LimitDegrees=45, swing2LimitDegrees=15, twistLimitDegrees=5}
 {bone1Name=dyn_loincloth_02_l, bone2Name=dyn_loincloth_01_l, swing1LimitDegrees=45, swing2LimitDegrees=15, twistLimitDegrees=5}
 {bone1Name=dyn_loincloth_03_l, bone2Name=dyn_loincloth_02_l, swing1LimitDegrees=45, swing2LimitDegrees=15, twistLimitDegrees=5}
 {bone1Name=dyn_loincloth_04_l, bone2Name=dyn_loincloth_03_l, swing1LimitDegrees=45, swing2LimitDegrees=15, twistLimitDegrees=5}
 {bone1Name=dyn_loincloth_05_l, bone2Name=dyn_loincloth_04_l, swing1LimitDegrees=45, swing2LimitDegrees=15, twistLimitDegrees=5}
]
