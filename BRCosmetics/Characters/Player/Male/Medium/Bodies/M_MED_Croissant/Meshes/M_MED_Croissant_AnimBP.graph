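# Post-process style AnimBP for the M_MED_Croissant male body that copies the parent mesh pose and simulates loincloth physics.
# Blends by crouch state between crouch and loincloth RigidBody assets, with gravity updated from skydive-aware override params through a deprecated function.
# A RestrictCape cosmetic tag drives a backpack socket offset before a deform Control Rig produces the output.
# Crouch-sprint loincloth tweaks and a leg-aim constraint sit unconnected to the output.
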AnimBlueprint /BRCosmetics/Characters/Player/Male/Medium/Bodies/M_MED_Croissant/Meshes/M_MED_Croissant_AnimBP
target_skeleton=&/M_MED_Croissant_Skeleton
graphs [
 AnimGraph {nodes [
   n1 = Get bIsCrouchSprinting()
   n2 = "Transform (Modify) Bone - Bone: dyn_loincloth_01"!(ComponentPose, bAlphaBoolEnabled:bool=@n1, Translation="0,0,0", Rotation="0.000,20,-0.000", Scale="1,1.000,0.6") -> (Pose) {
    alphaInputType=Bool
    rotationMode=BMM_Additive
    rotationSpace=BCS_BoneSpace
    scaleMode=BMM_Replace
    scaleSpace=BCS_BoneSpace
    alphaBoolBlend {blendOption=Linear}
    boneToModify {boneName=dyn_loincloth_01}
    rotation {yaw=20}
    scale {z=0.6}
   }
   "Transform (Modify) Bone - Bone: dyn_loincloth_02"!(ComponentPose=@n2, bAlphaBoolEnabled:bool=@n1, Translation="0,0,0", Rotation="0,20,-0.000", Scale="1,1.000,1.7") -> (Pose) {
    alphaInputType=Bool
    rotationMode=BMM_Additive
    rotationSpace=BCS_BoneSpace
    scaleMode=BMM_Replace
    scaleSpace=BCS_BoneSpace
    alphaBoolBlend {blendOption=Linear}
    boneToModify {boneName=dyn_loincloth_02}
    rotation {yaw=20}
    scale {z=1.7}
   }
   n3 = "Constraint - dyn_leg_aim (PP)"!(ComponentPose, Alpha=1, ConstraintWeights_0=1, ConstraintWeights_1:float=1) -> (Pose) {
    alphaBoolBlend {blendOption=Linear}
    boneToModify {boneName=dyn_leg_aim}
    constraintSetup [{targetBone {boneName=thigh_l}}, {targetBone {boneName=thigh_r}}]
    constraintWeights [1,1]
   }
   "Transform (Modify) Bone - Bone: dyn_loincloth_base"(ComponentPose=@n3, Alpha=1, Translation="0,0,0", Rotation="-10,31.712,-20.067", Scale="1,1.000,0.6") {
    rotationMode=BMM_Additive
    rotationSpace=BCS_BoneSpace
    scaleMode=BMM_Replace
    scaleSpace=BCS_BoneSpace
    alphaBoolBlend {blendOption=Linear}
    boneToModify {boneName=dyn_loincloth_base}
    rotation=<rot -10 31.712 -20.067>
    scale {z=0.6}
   }
   n4 = Get GravityOverride()
   n5 = RigidBody!(
    ComponentPose=Local To Component(LocalPose=Copy Pose From Mesh() {bUseAttachedParent=true})
    Alpha
    OverrideWorldGravity:Vector=@n4
    ExternalForce
   ) -> (Pose) {
    bOverrideWorldGravity=true
    overridePhysicsAsset=&/M_MED_Croissant_Crouch_Physics
    simulationSpace=BaseBoneSpace
    alphaBoolBlend {blendOption=Linear}
    baseBoneRef {boneName=pelvis}
    componentAppliedLinearAccClamp=<2000 2000 2000>
    componentLinearAccScale=<0.3 0.3 0.35>
    componentLinearVelScale=<0.3 0.3 0.35>
    overrideWorldGravity {z=-980}
   }
   n6 = Component To Local(ComponentPose=@n5)
   n7 = RigidBody(ComponentPose=Local To Component(LocalPose=Copy Pose From Mesh() {bUseAttachedParent=true}), OverrideWorldGravity=@n4) {
    bOverrideWorldGravity=true
    overridePhysicsAsset=&/M_MED_Croissant_Loincloth_Physics
    simulationSpace=BaseBoneSpace
    alphaBoolBlend {blendOption=Linear}
    baseBoneRef {boneName=pelvis}
    componentAppliedLinearAccClamp=<2000 2000 2000>
    componentLinearAccScale=<0.3 0.3 0.35>
    componentLinearVelScale=<0.3 0.3 0.35>
    overrideWorldGravity {z=-980}
   }
   n8 = "Transform (Modify) Bone - Bone: dyn_loincloth_base"(ComponentPose=@n7, Alpha=1, Translation="0,0,0", Rotation="0.000,0.000,-10.000", Scale="1,1,1") {
    rotationMode=BMM_Additive
    rotationSpace=BCS_BoneSpace
    scaleMode=BMM_Replace
    alphaBoolBlend {blendOption=Linear}
    boneToModify {boneName=dyn_loincloth_base}
    rotation=<rot 0 0 -10>
   }
   n9 = Component To Local(ComponentPose=@n8)
   n10 = Get bIsCrouching()
   n11 = "Transform (Modify) Bone - Bone: dyn_backpack_socket"!(
    ComponentPose=Local To Component(LocalPose=Blend Poses by bool(BlendPose_0=@n6, BlendPose_1=@n9, BlendTime_0=0.25, BlendTime_1=0.25, bActiveValue=@n10) {blendType=Linear, childUpateMode=ResetChildOnActivate})
    bAlphaBoolEnabled:bool=Get cape_active() -> (cape_active:bool)
    Translation="0.000,-3,0.000"
    Rotation="0.000,15,-0.000"
    Scale="1,1,1"
   ) -> (Pose) {
    alphaInputType=Bool
    rotationMode=BMM_Additive
    rotationSpace=BCS_BoneSpace
    translationMode=BMM_Additive
    translationSpace=BCS_BoneSpace
    alphaBoolBlend {blendOption=Linear}
    boneToModify {boneName=dyn_backpack_socket}
    rotation {yaw=15}
    translation {y=-3}
   }
   n12 = Control Rig(Source=Component To Local(ComponentPose=@n11)) {
    alphaBoolBlend {blendOption=Linear}
    controlRigAssetReference {blueprintRigClass=&/M_MED_Croissant_Deform_ControlRig.M_MED_Croissant_Deform_ControlRig_C}
   }
   Output Pose(Result=@n12) {layerGroup=None}
  ]}
 EventGraph {nodes [
   Event BlueprintUpdateAnimation()
   > Set cape_active(cape_active:bool=CharacterItemDefinitionHasAnyMetaTag(Tags=Make Array("[0]"=(TagName="Cosmetics.Character.RestrictCape")))) -> (Output_Get:bool)
   > n1 = Gravity Override State Function DEPRECATED(CharacterPartAnimInstance=Self-Reference(), GameTimeInSeconds=GetGameTimeInSeconds(), GravityOverrideParams=Get Gravity Override Params())
   > Set GravityOverride(GravityOverride=@n1)
  ]}
]
variables [
 cape_active {type=boolean}
 Gravity Override Params {
  type=GravityOverrideParamsStruct
  default {
   skydive_FloatingDown {gravityOverride=<0 -700 -600>}
   skydive_GoingUp {gravityOverride=<0 -300 -1500>}
   skydive_Parachuting {gravityOverride=<0 0 -980>, windAmplitude=<200 -200 400>}
  }
 }
 GravityOverride {type=Vector}
]
